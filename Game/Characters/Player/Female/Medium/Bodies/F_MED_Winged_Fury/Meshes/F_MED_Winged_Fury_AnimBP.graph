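# Post-process style AnimBlueprint for the Winged Fury player body, adding secondary motion to zippers, tongues and pouch.
# It copies the parent mesh pose, then chains constraints, additive corrective bones driven by thigh and foot deltas from ref pose, and spring-based AnimDynamics.
# Glove zipper simulation runs only when no Gameplay (gauntlet) part is found, despite the isGauntletOn name.
# Customization load hides the tail mesh under long backblings.
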AnimBlueprint /Game/Characters/Player/Female/Medium/Bodies/F_MED_Winged_Fury/Meshes/F_MED_Winged_Fury_AnimBP
target_skeleton=&/F_MED_Winged_Fury_Skeleton
graphs [
 AnimGraph {nodes [
   "Constraint - dyn_chest_zipper_base (PPP)"!(ComponentPose, Alpha=1, ConstraintWeights_0=0.7402, ConstraintWeights_1:float=0.217, ConstraintWeights_2:float=0.0428) -> (Pose) {
    alphaBoolBlend {blendOption=Linear}
    boneToModify {boneName=dyn_chest_zipper_base}
    constraintSetup [{targetBone {boneName=spine_04}}, {targetBone {boneName=spine_05}}, {targetBone {boneName=clavicle_l}}]
    constraintWeights [1,1,1]
   }
   "Constraint - dyn_glove_zipper_base_l (PPPPPPP)"!(
    ComponentPose
    Alpha=1
    ConstraintWeights_0=0.0001
    ConstraintWeights_1:float=0.0036
    ConstraintWeights_2:float=0.1517
    ConstraintWeights_3:float=0.0111
    ConstraintWeights_4:float=0.8245
    ConstraintWeights_5:float=0.0066
    ConstraintWeights_6:float=0.0023
   ) -> (Pose) {
    alphaBoolBlend {blendOption=Linear}
    boneToModify {boneName=dyn_glove_zipper_base_l}
    constraintSetup [
     {targetBone {boneName=middle_metacarpal_l}}
     {targetBone {boneName=pinky_metacarpal_l}}
     {targetBone {boneName=hand_l}}
     {targetBone {boneName=lowerarm_l}}
     {targetBone {boneName=lowerarm_twist_01_l}}
     {targetBone {boneName=lowerarm_twist_02_l}}
     {targetBone {boneName=ring_metacarpal_l}}
    ]
    constraintWeights [1,1,1,1,1,1,1]
   }
   n1 = "Constraint - thigh_volume_l (R)"(ComponentPose=Local To Component(LocalPose=Copy Pose From Mesh() {bUseAttachedParent=true}), Alpha=0.5, ConstraintWeights_0=1) {
    alphaBoolBlend {blendOption=Linear}
    boneToModify {boneName=thigh_volume_l}
    constraintSetup [{transformType=Rotation, targetBone {boneName=thigh_l}}]
    constraintWeights [1]
   }
   n2 = "Constraint - thigh_volume_r (R)"(ComponentPose=@n1, Alpha=0.5, ConstraintWeights_0=1) {
    alphaBoolBlend {blendOption=Linear}
    boneToModify {boneName=thigh_volume_r}
    constraintSetup [{transformType=Rotation, targetBone {boneName=thigh_r}}]
    constraintWeights [1]
   }
   n3 = "Transform (Modify) Bone - Bone: thigh_volume_l"(ComponentPose=@n2, Alpha=Get thigh_rot_l(), Translation="-1,-2,1", Rotation="0,0,0", Scale="1,1,1") {
    translationMode=BMM_Additive
    translationSpace=BCS_BoneSpace
    alphaBoolBlend {blendOption=Linear}
    boneToModify {boneName=thigh_volume_l}
    translation=<-1 -2 1>
   }
   n4 = "Transform (Modify) Bone - Bone: thigh_volume_r"(ComponentPose=@n3, Alpha=Get thigh_rot_r(), Translation="1,2,-1", Rotation="0,0,0", Scale="1,1,1") {
    translationMode=BMM_Additive
    translationSpace=BCS_BoneSpace
    alphaBoolBlend {blendOption=Linear}
    boneToModify {boneName=thigh_volume_r}
    translation=<1 2 -1>
   }
   n5 = "Constraint - dyn_shoulder_zipperA_base_l (PP)"!(ComponentPose=@n4, Alpha=1, ConstraintWeights_0=1, ConstraintWeights_1:float=0) -> (Pose) {
    alphaBoolBlend {blendOption=Linear}
    boneToModify {boneName=dyn_shoulder_zipperA_base_l}
    constraintSetup [{targetBone {boneName=upperarm_twist_01_l}}, {targetBone {boneName=clavicle_l}}]
    constraintWeights [1,1]
   }
   n6 = "Constraint - dyn_shoulder_zipperB_base_l (PPP)"!(ComponentPose=@n5, Alpha=1, ConstraintWeights_0=0.051, ConstraintWeights_1:float=0.043, ConstraintWeights_2:float=0.906) -> (Pose) {
    alphaBoolBlend {blendOption=Linear}
    boneToModify {boneName=dyn_shoulder_zipperB_base_l}
    constraintSetup [{targetBone {boneName=spine_04}}, {targetBone {boneName=spine_05}}, {targetBone {boneName=upperarm_twist_01_l}}]
    constraintWeights [1,1,1]
   }
   n7 = "Constraint - dyn_shoulder_zipperA_base_r (PP)"!(ComponentPose=@n6, Alpha=1, ConstraintWeights_0=0, ConstraintWeights_1:float=1) -> (Pose) {
    alphaBoolBlend {blendOption=Linear}
    boneToModify {boneName=dyn_shoulder_zipperA_base_r}
    constraintSetup [{targetBone {boneName=clavicle_r}}, {targetBone {boneName=upperarm_twist_01_r}}]
    constraintWeights [1,1]
   }
   n8 = "Constraint - dyn_shoulder_zipperB_base_r (PPP)"!(ComponentPose=@n7, Alpha=1, ConstraintWeights_0=0.906, ConstraintWeights_1:float=0.051, ConstraintWeights_2:float=0.043) -> (Pose) {
    alphaBoolBlend {blendOption=Linear}
    boneToModify {boneName=dyn_shoulder_zipperB_base_r}
    constraintSetup [{targetBone {boneName=upperarm_twist_01_r}}, {targetBone {boneName=spine_04}}, {targetBone {boneName=spine_05}}]
    constraintWeights [1,1,1]
   }
   n9 = "Constraint - dyn_torso_back_tongue_base_l (PPPPPP)"!(
    ComponentPose=@n8
    Alpha=1
    ConstraintWeights_0=0.0852
    ConstraintWeights_1:float=0.0013
    ConstraintWeights_2:float=0.0355
    ConstraintWeights_3:float=0.1261
    ConstraintWeights_4:float=0.68
    ConstraintWeights_5:float=0.072
   ) -> (Pose) {
    alphaBoolBlend {blendOption=Linear}
    boneToModify {boneName=dyn_torso_back_tongue_base_l}
    constraintSetup [
     {targetBone {boneName=clavicle_l}}
     {targetBone {boneName=spine_01}}
     {targetBone {boneName=spine_02}}
     {targetBone {boneName=spine_03}}
     {targetBone {boneName=spine_04}}
     {targetBone {boneName=spine_05}}
    ]
    constraintWeights [1,1,1,1,1,1]
   }
   n10 = "Constraint - dyn_torso_back_tongue_base_r (PPPPPP)"!(
    ComponentPose=@n9
    Alpha=1
    ConstraintWeights_0=0.0013
    ConstraintWeights_1:float=0.0355
    ConstraintWeights_2:float=0.1261
    ConstraintWeights_3:float=0.68
    ConstraintWeights_4:float=0.072
    ConstraintWeights_5:float=0.0852
   ) -> (Pose) {
    alphaBoolBlend {blendOption=Linear}
    boneToModify {boneName=dyn_torso_back_tongue_base_r}
    constraintSetup [
     {targetBone {boneName=spine_01}}
     {targetBone {boneName=spine_02}}
     {targetBone {boneName=spine_03}}
     {targetBone {boneName=spine_04}}
     {targetBone {boneName=spine_05}}
     {targetBone {boneName=clavicle_r}}
    ]
    constraintWeights [1,1,1,1,1,1]
   }
   n11 = "Constraint - dyn_chest_zipper_base (PP)"!(ComponentPose=@n10, Alpha=1, ConstraintWeights_0=0.7733, ConstraintWeights_1:float=0.2267) -> (Pose) {
    alphaBoolBlend {blendOption=Linear}
    boneToModify {boneName=dyn_chest_zipper_base}
    constraintSetup [{targetBone {boneName=spine_04}}, {targetBone {boneName=spine_05}}]
    constraintWeights [1,1]
   }
   n12 = "Constraint - dyn_torso_front_tongue_base (PPPP)"!(
    ComponentPose=@n11
    Alpha=1
    ConstraintWeights_0=0.0906
    ConstraintWeights_1:float=0.2085
    ConstraintWeights_2:float=0.224
    ConstraintWeights_3:float=0.477
   ) -> (Pose) {
    alphaBoolBlend {blendOption=Linear}
    boneToModify {boneName=dyn_torso_front_tongue_base}
    constraintSetup [{targetBone {boneName=spine_01}}, {targetBone {boneName=spine_02}}, {targetBone {boneName=spine_03}}, {targetBone {boneName=spine_04}}]
    constraintWeights [1,1,1,1]
   }
   n13 = "Constraint - dyn_glove_zipper_base_l (PPPPPPP)"(
    ComponentPose=@n12
    Alpha=1
    ConstraintWeights_0=0.0036
    ConstraintWeights_1=0.1517
    ConstraintWeights_2=0.0066
    ConstraintWeights_3=0.0001
    ConstraintWeights_4=0.8245
    ConstraintWeights_5=0.0111
    ConstraintWeights_6=0.0023
   ) {
    alphaBoolBlend {blendOption=Linear}
    boneToModify {boneName=dyn_glove_zipper_base_l}
    constraintSetup [
     {targetBone {boneName=pinky_metacarpal_l}}
     {targetBone {boneName=hand_l}}
     {targetBone {boneName=lowerarm_twist_02_l}}
     {targetBone {boneName=middle_metacarpal_l}}
     {targetBone {boneName=lowerarm_twist_01_l}}
     {targetBone {boneName=lowerarm_l}}
     {targetBone {boneName=ring_metacarpal_l}}
    ]
    constraintWeights [1,1,1,1,1,1,1]
   }
   n14 = "Constraint - dyn_glove_zipper_base_r (PPPPPP)"!(
    ComponentPose=@n13
    Alpha=1
    ConstraintWeights_0=0.0095
    ConstraintWeights_1:float=0.0002
    ConstraintWeights_2:float=0.016
    ConstraintWeights_3:float=0.8775
    ConstraintWeights_4:float=0.0959
    ConstraintWeights_5:float=0.0008
   ) -> (Pose) {
    alphaBoolBlend {blendOption=Linear}
    boneToModify {boneName=dyn_glove_zipper_base_r}
    constraintSetup [
     {targetBone {boneName=lowerarm_twist_02_r}}
     {targetBone {boneName=ring_metacarpal_r}}
     {targetBone {boneName=lowerarm_r}}
     {targetBone {boneName=lowerarm_twist_01_r}}
     {targetBone {boneName=hand_r}}
     {targetBone {boneName=pinky_metacarpal_r}}
    ]
    constraintWeights [1,1,1,1,1,1]
   }
   n15 = "Constraint - dyn_shoe_zipper_base_l (P)"!(ComponentPose=@n14, Alpha=1, ConstraintWeights_0=1, ConstraintWeights_1:float=0) -> (Pose) {
    alphaBoolBlend {blendOption=Linear}
    boneToModify {boneName=dyn_shoe_zipper_base_l}
    constraintSetup [{targetBone {boneName=foot_l}}]
    constraintWeights [1,0]
   }
   n16 = "Constraint - dyn_shoe_zipper_base_r (P)"!(ComponentPose=@n15, Alpha=1, ConstraintWeights_0=1, ConstraintWeights_1:float=0) -> (Pose) {
    alphaBoolBlend {blendOption=Linear}
    boneToModify {boneName=dyn_shoe_zipper_base_r}
    constraintSetup [{targetBone {boneName=foot_r}}]
    constraintWeights [1,0]
   }
   n17 = "Transform (Modify) Bone - Bone: dyn_shoe_tongue_base_l"(ComponentPose=@n16, Alpha=1, Translation="0,0,0", Rotation=Get foot_rot_l() -> (foot_rot_l:Rotator), Scale="1,1,1") {
    rotationMode=BMM_Additive
    rotationSpace=BCS_BoneSpace
    translationMode=BMM_Additive
    translationSpace=BCS_BoneSpace
    alphaBoolBlend {blendOption=Linear}
    boneToModify {boneName=dyn_shoe_tongue_base_l}
   }
   n18 = "Transform (Modify) Bone - Bone: dyn_shoe_tongue_base_r"(ComponentPose=@n17, Alpha=1, Translation="0,0,0", Rotation=Get foot_rot_r() -> (foot_rot_r:Rotator), Scale="1,1,1") {
    rotationMode=BMM_Additive
    rotationSpace=BCS_BoneSpace
    translationMode=BMM_Additive
    translationSpace=BCS_BoneSpace
    alphaBoolBlend {blendOption=Linear}
    boneToModify {boneName=dyn_shoe_tongue_base_r}
   }
   n19 = "Transform (Modify) Bone - Bone: dyn_pouch_base"(ComponentPose=@n18, Alpha=Get thigh_rot_l(), Translation="-1,-3,-0.6", Rotation="0,0,-30", Scale="1,1,1") {
    rotationMode=BMM_Additive
    rotationSpace=BCS_BoneSpace
    translationMode=BMM_Additive
    translationSpace=BCS_BoneSpace
    alphaBoolBlend {blendOption=Linear}
    alphaScaleBiasClamp {inRange {max=0, min=-100}, outRange {max=0, min=1}}
    boneToModify {boneName=dyn_pouch_base}
    rotation {roll=-30}
    translation=<-1 -3 -0.6>
   }
   n20 = "Transform (Modify) Bone - Bone: dyn_pouch_base"(ComponentPose=@n19, Alpha=Get thigh_sideRot_l() -> (thigh_sideRot_l:double), Translation="0,-2,0", Rotation="0,-20,-20", Scale="1,1,1") {
    rotationMode=BMM_Additive
    rotationSpace=BCS_BoneSpace
    translationMode=BMM_Additive
    translationSpace=BCS_BoneSpace
    alphaBoolBlend {blendOption=Linear}
    alphaScaleBiasClamp {bMapRange=true, inRange {max=40}}
    boneToModify {boneName=dyn_pouch_base}
    rotation {roll=-20, yaw=-20}
    translation {y=-2}
   }
   n21 = Anim Dynamics - Bone: dyn_shoulder_zipperA_l(ComponentPose=@n20, Alpha=1, ExternalForce="0,0,0") {
    angularDampingOverride=1
    angularSpringConstant=1500
    bAngularSpring=true
    bGravityOverrideInSimSpace=true
    bOverrideAngularDamping=true
    gravityScale=0.35
    simulationSpace=RootRelative
    alphaBoolBlend {blendOption=Linear}
    boundBone {boneName=dyn_shoulder_zipperA_l}
    componentLinearAccScale=<0.3 0.3 0.5>
    componentLinearVelScale=<0.3 0.3 0.5>
    physicsBodyDefinitions [{
      sphereCollisionRadius=0
      boundBone {boneName=dyn_shoulder_zipperA_l}
      boxExtents=<10 20 20>
      constraintSetup {angularLimitsMax=<0 5 15>, angularLimitsMin=<0 -50 -15>, angularTarget=<0 0 0>}
      localJointOffset=<10 0 0>
     }]
   }
   n22 = Anim Dynamics - Bone: dyn_shoulder_zipperB_l(ComponentPose=@n21, Alpha=1, ExternalForce="0,0,0") {
    angularDampingOverride=1
    angularSpringConstant=1500
    bAngularSpring=true
    bGravityOverrideInSimSpace=true
    bOverrideAngularDamping=true
    gravityScale=0.35
    simulationSpace=RootRelative
    alphaBoolBlend {blendOption=Linear}
    boundBone {boneName=dyn_shoulder_zipperB_l}
    componentLinearAccScale=<0.3 0.3 0.5>
    componentLinearVelScale=<0.3 0.3 0.5>
    physicsBodyDefinitions [{
      sphereCollisionRadius=0
      boundBone {boneName=dyn_shoulder_zipperB_l}
      boxExtents=<10 20 20>
      constraintSetup {angularLimitsMax=<0 40 10>, angularLimitsMin=<0 -40 -50>, angularTarget=<0 0 0>}
      localJointOffset=<10 0 0>
     }]
   }
   n23 = Get isGauntletOn() -> (isGauntletOn:bool)
   n24 = Anim Dynamics - Bone: dyn_glove_zipper_l!(ComponentPose=@n22, bAlphaBoolEnabled:bool=@n23, ExternalForce="0,0,0") -> (Pose) {
    alphaInputType=Bool
    angularDampingOverride=1
    angularSpringConstant=1500
    bAngularSpring=true
    bGravityOverrideInSimSpace=true
    bOverrideAngularDamping=true
    gravityScale=0.35
    simulationSpace=RootRelative
    alphaBoolBlend {blendOption=Linear}
    boundBone {boneName=dyn_glove_zipper_l}
    componentLinearAccScale=<0.3 0.3 0.5>
    componentLinearVelScale=<0.3 0.3 0.5>
    physicsBodyDefinitions [{
      sphereCollisionRadius=0
      boundBone {boneName=dyn_glove_zipper_l}
      boxExtents=<10 20 20>
      constraintSetup {angularLimitsMax=<0 40 10>, angularLimitsMin=<0 -40 -80>, angularTarget=<0 0 0>}
      localJointOffset=<10 0 0>
     }]
   }
   n25 = Anim Dynamics - Bone: dyn_shoulder_zipperA_r(ComponentPose=@n24, Alpha=1, ExternalForce="0,0,0") {
    angularDampingOverride=1
    angularSpringConstant=1500
    bAngularSpring=true
    bGravityOverrideInSimSpace=true
    bOverrideAngularDamping=true
    gravityScale=0.35
    simulationSpace=RootRelative
    alphaBoolBlend {blendOption=Linear}
    boundBone {boneName=dyn_shoulder_zipperA_r}
    componentLinearAccScale=<0.3 0.3 0.5>
    componentLinearVelScale=<0.3 0.3 0.5>
    physicsBodyDefinitions [{
      sphereCollisionRadius=0
      boundBone {boneName=dyn_shoulder_zipperA_r}
      boxExtents=<10 20 20>
      constraintSetup {angularLimitsMax=<0 5 15>, angularLimitsMin=<0 -50 -15>, angularTarget=<0 0 0>}
      localJointOffset=<-10 0 0>
     }]
   }
   n26 = Anim Dynamics - Bone: dyn_shoulder_zipperB_r(ComponentPose=@n25, Alpha=1, ExternalForce="0,0,0") {
    angularDampingOverride=1
    angularSpringConstant=1500
    bAngularSpring=true
    bGravityOverrideInSimSpace=true
    bOverrideAngularDamping=true
    gravityScale=0.35
    simulationSpace=RootRelative
    alphaBoolBlend {blendOption=Linear}
    boundBone {boneName=dyn_shoulder_zipperB_r}
    componentLinearAccScale=<0.3 0.3 0.5>
    componentLinearVelScale=<0.3 0.3 0.5>
    physicsBodyDefinitions [{
      sphereCollisionRadius=0
      boundBone {boneName=dyn_shoulder_zipperB_r}
      boxExtents=<10 20 20>
      constraintSetup {angularLimitsMax=<0 40 10>, angularLimitsMin=<0 -40 -50>, angularTarget=<0 0 0>}
      localJointOffset=<-10 0 0>
     }]
   }
   n27 = Anim Dynamics - Bone: dyn_glove_zipper_r!(ComponentPose=@n26, bAlphaBoolEnabled:bool=@n23, ExternalForce="0,0,0") -> (Pose) {
    alphaInputType=Bool
    angularDampingOverride=1
    angularSpringConstant=1500
    bAngularSpring=true
    bGravityOverrideInSimSpace=true
    bOverrideAngularDamping=true
    gravityScale=0.35
    simulationSpace=RootRelative
    alphaBoolBlend {blendOption=Linear}
    boundBone {boneName=dyn_glove_zipper_r}
    componentLinearAccScale=<0.3 0.3 0.5>
    componentLinearVelScale=<0.3 0.3 0.5>
    physicsBodyDefinitions [{
      sphereCollisionRadius=0
      boundBone {boneName=dyn_glove_zipper_r}
      boxExtents=<10 20 20>
      constraintSetup {angularLimitsMax=<0 40 10>, angularLimitsMin=<0 -40 -80>, angularTarget=<0 0 0>}
      localJointOffset=<-10 0 0>
     }]
   }
   n28 = Anim Dynamics - Bone: dyn_torso_back_tongue_l(ComponentPose=@n27, Alpha=1, ExternalForce="0,0,0") {
    angularDampingOverride=1
    angularSpringConstant=1500
    bAngularSpring=true
    bGravityOverrideInSimSpace=true
    bOverrideAngularDamping=true
    gravityScale=0.35
    simulationSpace=RootRelative
    alphaBoolBlend {blendOption=Linear}
    boundBone {boneName=dyn_torso_back_tongue_l}
    componentLinearAccScale=<0.3 0.3 0.5>
    componentLinearVelScale=<0.3 0.3 0.5>
    physicsBodyDefinitions [{
      sphereCollisionRadius=0
      boundBone {boneName=dyn_torso_back_tongue_l}
      boxExtents=<10 20 20>
      constraintSetup {angularLimitsMax=<0 15 5>, angularLimitsMin=<0 -15 -5>, angularTarget=<0 0 0>}
      localJointOffset=<10 0 0>
     }]
   }
   n29 = Anim Dynamics - Bone: dyn_torso_back_tongue_r(ComponentPose=@n28, Alpha=1, ExternalForce="0,0,0") {
    angularDampingOverride=1
    angularSpringConstant=1500
    bAngularSpring=true
    bGravityOverrideInSimSpace=true
    bOverrideAngularDamping=true
    gravityScale=0.35
    simulationSpace=RootRelative
    alphaBoolBlend {blendOption=Linear}
    boundBone {boneName=dyn_torso_back_tongue_r}
    componentLinearAccScale=<0.3 0.3 0.5>
    componentLinearVelScale=<0.3 0.3 0.5>
    physicsBodyDefinitions [{
      sphereCollisionRadius=0
      boundBone {boneName=dyn_torso_back_tongue_r}
      boxExtents=<10 20 20>
      constraintSetup {angularLimitsMax=<0 15 5>, angularLimitsMin=<0 -15 -5>, angularTarget=<0 0 0>}
      localJointOffset=<-10 0 0>
     }]
   }
   n30 = Anim Dynamics - Bone: dyn_chest_zipper(ComponentPose=@n29, Alpha=1, ExternalForce="0,0,0") {
    angularDampingOverride=1
    angularSpringConstant=1500
    bAngularSpring=true
    bGravityOverrideInSimSpace=true
    bOverrideAngularDamping=true
    gravityScale=0.35
    simulationSpace=RootRelative
    alphaBoolBlend {blendOption=Linear}
    boundBone {boneName=dyn_chest_zipper}
    componentLinearAccScale=<0.3 0.3 0.5>
    componentLinearVelScale=<0.3 0.3 0.5>
    physicsBodyDefinitions [{
      sphereCollisionRadius=0
      boundBone {boneName=dyn_chest_zipper}
      boxExtents=<10 20 20>
      constraintSetup {angularLimitsMax=<0 25 50>, angularLimitsMin=<0 -25 0>, angularTarget=<0 0 0>}
      localJointOffset=<10 0 0>
     }]
   }
   n31 = Anim Dynamics - Bone: dyn_pouch_zipper(ComponentPose=@n30, Alpha=1, ExternalForce="0,0,0") {
    angularDampingOverride=1
    angularSpringConstant=1500
    bAngularSpring=true
    bGravityOverrideInSimSpace=true
    bOverrideAngularDamping=true
    gravityScale=0.35
    simulationSpace=RootRelative
    alphaBoolBlend {blendOption=Linear}
    boundBone {boneName=dyn_pouch_zipper}
    componentLinearAccScale=<0.3 0.3 0.5>
    componentLinearVelScale=<0.3 0.3 0.5>
    physicsBodyDefinitions [{
      sphereCollisionRadius=0
      boundBone {boneName=dyn_pouch_zipper}
      boxExtents=<10 20 20>
      constraintSetup {angularLimitsMax=<0 20 0>, angularLimitsMin=<0 -20 -90>, angularTarget=<0 0 0>}
      localJointOffset=<10 0 0>
     }]
   }
   n32 = Anim Dynamics - Bone: dyn_torso_front_tongue(ComponentPose=@n31, Alpha=1, ExternalForce="0,0,0") {
    angularDampingOverride=1
    angularSpringConstant=1500
    bAngularSpring=true
    bGravityOverrideInSimSpace=true
    bOverrideAngularDamping=true
    gravityScale=0.35
    simulationSpace=RootRelative
    alphaBoolBlend {blendOption=Linear}
    boundBone {boneName=dyn_torso_front_tongue}
    componentLinearAccScale=<0.3 0.3 0.5>
    componentLinearVelScale=<0.3 0.3 0.5>
    physicsBodyDefinitions [{
      sphereCollisionRadius=0
      boundBone {boneName=dyn_torso_front_tongue}
      boxExtents=<20 10 20>
      constraintSetup {angularTargetAxis=AxisY, twistAxis=AxisY, angularLimitsMax=<5 15 0>, angularLimitsMin=<-5 -15 0>, angularTarget=<0 0 0>}
      localJointOffset=<0 10 0>
     }]
   }
   n33 = Anim Dynamics - Bone: dyn_shoe_zipper_l(ComponentPose=@n32, Alpha=1, ExternalForce="0,0,0") {
    angularDampingOverride=1
    angularSpringConstant=1500
    bAngularSpring=true
    bGravityOverrideInSimSpace=true
    bOverrideAngularDamping=true
    gravityScale=0.35
    simulationSpace=RootRelative
    alphaBoolBlend {blendOption=Linear}
    boundBone {boneName=dyn_shoe_zipper_l}
    componentLinearAccScale=<0.3 0.3 0.5>
    componentLinearVelScale=<0.3 0.3 0.5>
    physicsBodyDefinitions [{
      sphereCollisionRadius=0
      boundBone {boneName=dyn_shoe_zipper_l}
      boxExtents=<10 20 20>
      constraintSetup {angularLimitsMax=<0 5 15>, angularLimitsMin=<0 -50 -15>, angularTarget=<0 0 0>}
      localJointOffset=<10 0 0>
     }]
   }
   n34 = Anim Dynamics - Bone: dyn_shoe_zipper_r(ComponentPose=@n33, Alpha=1, ExternalForce="0,0,0") {
    angularDampingOverride=1
    angularSpringConstant=1500
    bAngularSpring=true
    bGravityOverrideInSimSpace=true
    bOverrideAngularDamping=true
    gravityScale=0.35
    simulationSpace=RootRelative
    alphaBoolBlend {blendOption=Linear}
    boundBone {boneName=dyn_shoe_zipper_r}
    componentLinearAccScale=<0.3 0.3 0.5>
    componentLinearVelScale=<0.3 0.3 0.5>
    physicsBodyDefinitions [{
      sphereCollisionRadius=0
      boundBone {boneName=dyn_shoe_zipper_r}
      boxExtents=<10 20 20>
      constraintSetup {angularLimitsMax=<0 5 15>, angularLimitsMin=<0 -50 -15>, angularTarget=<0 0 0>}
      localJointOffset=<-10 0 0>
     }]
   }
   n35 = Anim Dynamics - Bone: dyn_shoe_tongue_l(ComponentPose=@n34, Alpha=1, ExternalForce="0,0,0") {
    angularDampingOverride=1
    angularSpringConstant=1500
    bAngularSpring=true
    bGravityOverrideInSimSpace=true
    bOverrideAngularDamping=true
    gravityScale=0.35
    simulationSpace=RootRelative
    alphaBoolBlend {blendOption=Linear}
    boundBone {boneName=dyn_shoe_tongue_l}
    componentLinearAccScale=<0.3 0.3 0.5>
    componentLinearVelScale=<0.3 0.3 0.5>
    physicsBodyDefinitions [{
      sphereCollisionRadius=0
      boundBone {boneName=dyn_shoe_tongue_l}
      boxExtents=<10 20 20>
      constraintSetup {angularLimitsMax=<0 0 0>, angularLimitsMin=<0 0 -10>, angularTarget=<0 0 0>}
      localJointOffset=<5 0 0>
     }]
   }
   n36 = Anim Dynamics - Bone: dyn_shoe_tongue_r(ComponentPose=@n35, Alpha=1, ExternalForce="0,0,0") {
    angularDampingOverride=1
    angularSpringConstant=1500
    bAngularSpring=true
    bGravityOverrideInSimSpace=true
    bOverrideAngularDamping=true
    gravityScale=0.35
    simulationSpace=RootRelative
    alphaBoolBlend {blendOption=Linear}
    boundBone {boneName=dyn_shoe_tongue_r}
    componentLinearAccScale=<0.3 0.3 0.5>
    componentLinearVelScale=<0.3 0.3 0.5>
    physicsBodyDefinitions [{
      sphereCollisionRadius=0
      boundBone {boneName=dyn_shoe_tongue_r}
      boxExtents=<10 20 20>
      constraintSetup {angularLimitsMax=<0 0 0>, angularLimitsMin=<0 0 -10>, angularTarget=<0 0 0>}
      localJointOffset=<-5 0 0>
     }]
   }
   Output Pose(Result=Component To Local(ComponentPose=@n36)) {layerGroup=None}
  ]}
 EventGraph {nodes [
   Event BlueprintUpdateAnimation()
   > n1 = Sequence!(execute) -> (then_0, then_1, then_2:Exec, then_3:Exec)
   n2 = Sequence!(execute=@n1.then_2) -> (then_0, then_1, then_2)
   n3 = GetOwnerSkeletalMeshComponent()
   n4 = BreakRotator(InRot=BreakTransform(InTransform=GetDeltaTransformFromRefPose(self=@n3, BoneName=foot_l, BaseName=calf_l)).Rotation)
   > n5 = Branch(Condition="float < float"!(A:double=@n4.Yaw, B:double=0) -> (ReturnValue:bool) {is=title K2Node_CallFunction})
   > Set foot_rot_l(foot_rot_l:Rotator=MakeRotator(Yaw=float * float(A=@n4.Yaw, B=2))) -> (Output_Get:Rotator)
   Set foot_rot_l(execute=@n5.else, foot_rot_l="0, 0, 0")
   n6 = BreakRotator(InRot=BreakTransform(InTransform=GetDeltaTransformFromRefPose(self=@n3, BoneName=foot_r, BaseName=calf_r)).Rotation)
   n7 = Branch(execute=@n2.then_1, Condition="float < float"(A=@n6.Yaw, B=0))
   > Set foot_rot_r(foot_rot_r:Rotator=MakeRotator(Yaw=float * float(A=@n6.Yaw, B=2))) -> (Output_Get:Rotator)
   Set foot_rot_r(execute=@n7.else, foot_rot_r="0, 0, 0")
   n8 = Get!(execute:Exec=@n1.then_3) -> (then:Exec, else:Exec, Gauntlet:Custom Character Part Object Reference)
   > Set isGauntletOn(isGauntletOn:bool=false) -> (Output_Get:bool)
   Set isGauntletOn(execute=@n8.else, isGauntletOn=true)
   n9 = GetOwnerSkeletalMeshComponent()
   n10 = BreakTransform(InTransform=GetDeltaTransformFromRefPose(self=@n9, BoneName=thigh_r, BaseName=pelvis))
   Set thigh_rot_r(execute=@n1.then_1, thigh_rot_r=MapRangeClamped(Value=BreakRotator(InRot=@n10.Rotation).Yaw, InRangeA=0, InRangeB=-100))
   n11 = BreakRotator(InRot=BreakTransform(InTransform=GetDeltaTransformFromRefPose(self=@n9, BoneName=thigh_l, BaseName=pelvis)).Rotation)
   Set thigh_rot_l(execute=@n1.then_0, thigh_rot_l=MapRangeClamped(Value=@n11.Yaw, InRangeA=0, InRangeB=-100))
   > Set thigh_sideRot_l(thigh_sideRot_l:double=@n11.Pitch) -> (Output_Get:double)
   n12 = Event OnCharacterCustomizationLoaded()
   > Set Gauntlet(Gauntlet:Custom Character Part Object Reference=Get Array Item with EFort Custom Part Type(Array:Array of Custom Character Part Object References=@n12.CustomParts, PartType=Gameplay) -> (Output:Custom Character Part Object Reference)) -> (Output_Get:Custom Character Part Object Reference)
   > Set BodyMesh(BodyMesh=GetOwnerSkeletalMeshComponent())
   > n13 = Cast To PlayerPawn_Athena(Object=TryGetPawnOwner())
   > n14 = Set PawnOwner(PawnOwner:Player Pawn Athena Object Reference=@n13) -> (Output_Get:Player Pawn Athena Object Reference)
   > Set TailMesh(TailMesh:Skeletal Mesh Component Object Reference=GetSkeletalMeshForPartType(self=@n14, PartType=MiscOrTail)) -> (Output_Get:Skeletal Mesh Component Object Reference)
   n15 = CheckForBackbling(LongBackblingThreshold=40)
   > Set isWearingBackbling(isWearingBackbling=@n15)
   > Set isWearingCape(isWearingCape=@n15.bIsBackblingLong)
   > Set characterHasTail(characterHasTail:bool=CheckForTail!(self:Custom Character Part Anim Instance Object Reference) -> (bHasTail:bool)) -> (Output_Get:bool)
   > Branch(Condition=Get characterHasTail() -> (characterHasTail:bool))
   > n16 = Branch(Condition=Get isWearingCape() -> (isWearingCape:bool))
   n17 = Get TailMesh() -> (TailMesh:Skeletal Mesh Component Object Reference)
   > SetVisibility!(execute, self:Scene Component Object Reference=@n17, bNewVisibility:bool=false, bPropagateToChildren:bool=false) -> (then)
   SetVisibility(execute=@n16.else, self=@n17, bNewVisibility=true, bPropagateToChildren=false)
   TryGetPawnOwner()
  ]}
]
variables [
 BodyMesh {type=SkeletalMeshComponent}
 characterHasTail {type=boolean}
 foot_rot_l {type=Rotator}
 foot_rot_r {type=Rotator}
 Gauntlet {type=CustomCharacterPart, default=null}
 isGauntletOn {type=boolean}
 isWearingBackbling {type=boolean}
 isWearingCape {type=boolean}
 PawnOwner {type=PlayerPawn_Athena_C}
 TailMesh {type=SkeletalMeshComponent}
 thigh_rot_l {type=number}
 thigh_rot_r {type=number}
 thigh_sideRot_l {type=number}
]
